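# Sound cue for the Ostrich mech vehicle's procedural foley, playing one of four large Set B servo sounds.
# An equally weighted Random node picks the wave, which passes through a Modulator left at its default settings.
# The cue uses the Ostrich emote attenuation and is routed to the emote music sound class, with a slight volume boost.
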SoundCue /Game/Athena/Sounds/Vehicles/Ostrich/ProceduralFoley/Ostrich_SetB_Large_ProcFoley_Cue
root=Modulator_0
nodes [
 {id=Modulator_0, children [Random_0]}
 {id=Random_0, children [WavePlayer_1,WavePlayer_2,WavePlayer_3,WavePlayer_4], properties {weights [1,1,1,1]}}
 {
  id=WavePlayer_1
  properties {soundWaveAssetPtr=&/Mech_Vehicle_Prodedural_Servo_SetB_Large_01}
 }
 {
  id=WavePlayer_2
  properties {soundWaveAssetPtr=&/Mech_Vehicle_Prodedural_Servo_SetB_Large_02}
 }
 {
  id=WavePlayer_3
  properties {soundWaveAssetPtr=&/Mech_Vehicle_Prodedural_Servo_SetB_Large_03}
 }
 {
  id=WavePlayer_4
  properties {soundWaveAssetPtr=&/Mech_Vehicle_Prodedural_Servo_SetB_Large_04}
 }
]
properties {
 attenuationSettings=/Game/Sounds/Attenuation/Fort_Emotes_Ostrich_Athena_Att
 maxDistance=6024
 soundClassObject=&Fort_Music_Emote
 volumeMultiplier=1.1
}
