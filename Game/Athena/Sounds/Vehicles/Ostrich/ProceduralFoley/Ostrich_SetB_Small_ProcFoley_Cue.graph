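# Sound cue for small Set B procedural servo foley on the Ostrich mech vehicle.
# A Random node picks among six equally weighted servo waves and feeds a Modulator that varies volume on each play.
# It uses a dedicated Ostrich foley concurrency set, but its attenuation is an Ostrich emote asset and its sound class is Fort_Music_Emote.
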SoundCue /Game/Athena/Sounds/Vehicles/Ostrich/ProceduralFoley/Ostrich_SetB_Small_ProcFoley_Cue
root=Modulator_0
nodes [
 {id=Modulator_0, children [Random_2], properties {volumeMin=0.7}}
 {
  id=Random_2
  children [WavePlayer_20, WavePlayer_21, WavePlayer_22, WavePlayer_23, WavePlayer_24, WavePlayer_28]
  properties {weights [1,1,1,1,1,1]}
 }
 {
  id=WavePlayer_20
  properties {soundWaveAssetPtr=&/Mech_Vehicle_Prodedural_Servo_SetB_Small_10}
 }
 {
  id=WavePlayer_21
  properties {soundWaveAssetPtr=&/Mech_Vehicle_Prodedural_Servo_SetB_Small_01}
 }
 {
  id=WavePlayer_22
  properties {soundWaveAssetPtr=&/Mech_Vehicle_Prodedural_Servo_SetB_Small_02}
 }
 {
  id=WavePlayer_23
  properties {soundWaveAssetPtr=&/Mech_Vehicle_Prodedural_Servo_SetB_Small_03}
 }
 {
  id=WavePlayer_24
  properties {soundWaveAssetPtr=&/Mech_Vehicle_Prodedural_Servo_SetB_Small_04}
 }
 {
  id=WavePlayer_28
  properties {soundWaveAssetPtr=&/Mech_Vehicle_Prodedural_Servo_SetB_Small_08}
 }
]
properties {
 attenuationSettings=/Game/Sounds/Attenuation/Fort_Emotes_Ostrich_Athena_Att
 maxDistance=6024
 soundClassObject=&Fort_Music_Emote
 volumeMultiplier=1.1
 concurrencySet [&/Ostrich_ProcFoley_Conc]
}
