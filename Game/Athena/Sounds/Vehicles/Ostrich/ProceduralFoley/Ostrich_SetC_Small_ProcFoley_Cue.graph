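# Sound cue for small Set C servo sounds in the Ostrich mech vehicle's procedural foley.
# A Random node picks one of five servo waves with equal weight, and a Modulator above it varies pitch and volume on each play.
# It uses the Ostrich procedural foley concurrency, but its sound class is the emote music class and its attenuation is an Ostrich emote asset.
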SoundCue /Game/Athena/Sounds/Vehicles/Ostrich/ProceduralFoley/Ostrich_SetC_Small_ProcFoley_Cue
root=Modulator_0
nodes [
 {id=Modulator_0, children [Random_0], properties {pitchMax=1.1, pitchMin=0.9, volumeMin=0.7}}
 {id=Random_0, children [WavePlayer_1, WavePlayer_2, WavePlayer_3, WavePlayer_7, WavePlayer_10], properties {weights [1,1,1,1,1]}}
 {
  id=WavePlayer_1
  properties {soundWaveAssetPtr=&/Mech_Vehicle_Prodedural_Servo_SetC_Small_10}
 }
 {
  id=WavePlayer_10
  properties {soundWaveAssetPtr=&/Mech_Vehicle_Prodedural_Servo_SetC_Small_09}
 }
 {
  id=WavePlayer_2
  properties {soundWaveAssetPtr=&/Mech_Vehicle_Prodedural_Servo_SetC_Small_01}
 }
 {
  id=WavePlayer_3
  properties {soundWaveAssetPtr=&/Mech_Vehicle_Prodedural_Servo_SetC_Small_02}
 }
 {
  id=WavePlayer_7
  properties {soundWaveAssetPtr=&/Mech_Vehicle_Prodedural_Servo_SetC_Small_06}
 }
]
properties {
 attenuationSettings=/Game/Sounds/Attenuation/Fort_Emotes_Ostrich_Athena_Att
 maxDistance=6024
 soundClassObject=&Fort_Music_Emote
 volumeMultiplier=1.1
 concurrencySet [&/Ostrich_ProcFoley_Conc]
}
