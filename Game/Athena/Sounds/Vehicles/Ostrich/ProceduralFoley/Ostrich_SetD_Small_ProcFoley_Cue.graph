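# Sound cue for small Set D procedural servo foley on the Ostrich mech vehicle.
# A Random node picks among four equally weighted servo waves, then a Modulator varies each play with a lowered minimum volume.
# The cue uses the Ostrich procedural foley concurrency set, but routes through an emote music sound class and an Ostrich emote attenuation asset.
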SoundCue /Game/Athena/Sounds/Vehicles/Ostrich/ProceduralFoley/Ostrich_SetD_Small_ProcFoley_Cue
root=Modulator_0
nodes [
 {id=Modulator_0, children [Random_0], properties {volumeMin=0.7}}
 {id=Random_0, children [WavePlayer_0,WavePlayer_1,WavePlayer_2,WavePlayer_3], properties {weights [1,1,1,1]}}
 {
  id=WavePlayer_0
  properties {soundWaveAssetPtr=&/Mech_Vehicle_Prodedural_Servo_SetD_Small_01}
 }
 {
  id=WavePlayer_1
  properties {soundWaveAssetPtr=&/Mech_Vehicle_Prodedural_Servo_SetD_Small_02}
 }
 {
  id=WavePlayer_2
  properties {soundWaveAssetPtr=&/Mech_Vehicle_Prodedural_Servo_SetD_Small_03}
 }
 {
  id=WavePlayer_3
  properties {soundWaveAssetPtr=&/Mech_Vehicle_Prodedural_Servo_SetD_Small_04}
 }
]
properties {
 attenuationSettings=/Game/Sounds/Attenuation/Fort_Emotes_Ostrich_Athena_Att
 maxDistance=6024
 soundClassObject=&Fort_Music_Emote
 volumeMultiplier=1.1
 concurrencySet [&/Ostrich_ProcFoley_Conc]
}
